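# Physics asset for the M_MED_Football_20 medium male player body cosmetic, set up for dangling shoelace physics.
# It holds one large capsule on each foot and small capsules on the dynamic lace bones, with no other body parts covered.
# Each foot anchors two separate two-link lace chains (00 to 01, 03 to 04).
# All joints share moderate swing limits and almost no twist, so laces bend without spinning.
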PhysicsAsset /BRCosmetics/Characters/Player/Male/Medium/Bodies/M_MED_Football_20/Meshes/M_MED_Football_20_Physics
bodies {
 dyn_lace_l_00 [{shapeType=Capsule, length=1, radius=1, center=<1.3312 -0.0763 -0.5931>, rotation=<rot 0 89.0568 -76.6488>}]
 dyn_lace_l_01 [{shapeType=Capsule, length=1, radius=1, center=<1.3194 -0.1322 -0.5943>, rotation=<rot 0 89.0568 -76.6488>}]
 dyn_lace_l_03 [{shapeType=Capsule, length=1, radius=1, center=<1.3199 -0.0677 -0.594>, rotation=<rot 0 89.0568 -76.6488>}]
 dyn_lace_l_04 [{shapeType=Capsule, length=1, radius=1, center=<1.31 -0.0696 -0.5961>, rotation=<rot 0 89.0568 -76.6488>}]
 dyn_lace_r_00 [{shapeType=Capsule, length=1, radius=1, center=<1.3526 -0.0662 -0.9337>, rotation=<rot 0 89.0568 -86.6486>}]
 dyn_lace_r_01 [{shapeType=Capsule, length=1, radius=1, center=<1.3526 -0.0662 -0.9337>, rotation=<rot 0 89.0568 -86.6486>}]
 dyn_lace_r_03 [{shapeType=Capsule, length=1, radius=1, center=<1.3526 -0.0662 -0.9337>, rotation=<rot 0 89.0568 -86.6486>}]
 dyn_lace_r_04 [{shapeType=Capsule, length=1, radius=1, center=<1.3526 -0.0662 -0.9337>, rotation=<rot 0 89.0568 -86.6486>}]
 foot_l [{shapeType=Capsule, length=14.4468, radius=4.9911, center=<-2.8608 4.9318 -0.8367>, rotation=<rot -0.5819 34.8443 -86.5777>}]
 foot_r [{shapeType=Capsule, length=14.4468, radius=4.9911, center=<2.5974 -4.903 -0.1559>, rotation=<rot -0.582 34.8433 -86.5777>}]
}
constraints [
 {bone1Name=dyn_lace_r_04, bone2Name=dyn_lace_r_03, swing1LimitDegrees=45, swing2LimitDegrees=45, twistLimitDegrees=1}
 {bone1Name=dyn_lace_r_01, bone2Name=dyn_lace_r_00, swing1LimitDegrees=45, swing2LimitDegrees=45, twistLimitDegrees=1}
 {bone1Name=dyn_lace_r_00, bone2Name=foot_r, swing1LimitDegrees=45, swing2LimitDegrees=45, twistLimitDegrees=1}
 {bone1Name=dyn_lace_r_03, bone2Name=foot_r, swing1LimitDegrees=45, swing2LimitDegrees=45, twistLimitDegrees=1}
 {bone1Name=dyn_lace_l_04, bone2Name=dyn_lace_l_03, swing1LimitDegrees=45, swing2LimitDegrees=45, twistLimitDegrees=1}
 {bone1Name=dyn_lace_l_01, bone2Name=dyn_lace_l_00, swing1LimitDegrees=45, swing2LimitDegrees=45, twistLimitDegrees=1}
 {bone1Name=dyn_lace_l_03, bone2Name=foot_l, swing1LimitDegrees=45, swing2LimitDegrees=45, twistLimitDegrees=1}
 {bone1Name=dyn_lace_l_00, bone2Name=foot_l, swing1LimitDegrees=45, swing2LimitDegrees=45, twistLimitDegrees=1}
]
