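# Post-process style AnimBP for the M_MED_Football_20 male body, layering corrective and physics motion onto the pose copied from the attached parent mesh.
# Each update derives left and right alphas from thigh rotation against the pelvis, driving additive offsets on dyn_vol bones.
# Two chained RigidBody nodes then simulate in root-bone and world space at fixed 0.25 alpha.
# Front-end-dependent dynalpha and rigidalpha are set but unused in the AnimGraph.
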AnimBlueprint /BRCosmetics/Characters/Player/Male/Medium/Bodies/M_MED_Football_20/Meshes/M_MED_Football_20_Skeleton_AnimBP
target_skeleton=&/M_MED_Football_20_Skeleton
graphs [
 AnimGraph {nodes [
   n1 = "Transform (Modify) Bone - Bone: dyn_vol_l"(
    ComponentPose=Local To Component(LocalPose=Copy Pose From Mesh() {bUseAttachedParent=true})
    Alpha=Get thigh_rot_l()
    Translation="2.86,-6.774,0.409"
    Rotation="0,0,0"
    Scale="1,1,1"
   ) {
    rotationSpace=BCS_BoneSpace
    scaleMode=BMM_Additive
    scaleSpace=BCS_BoneSpace
    translationMode=BMM_Additive
    translationSpace=BCS_BoneSpace
    alphaBoolBlend {blendOption=Linear}
    boneToModify {boneName=dyn_vol_l}
    translation=<2.86 -6.774 0.409>
   }
   n2 = "Transform (Modify) Bone - Bone: dyn_vol_r"(ComponentPose=@n1, Alpha=Get thigh_rot_r(), Translation="2.860,5.043,0.000", Rotation="0,0,0", Scale="1,1,1") {
    rotationSpace=BCS_BoneSpace
    scaleMode=BMM_Additive
    scaleSpace=BCS_BoneSpace
    translationMode=BMM_Additive
    translationSpace=BCS_BoneSpace
    alphaBoolBlend {blendOption=Linear}
    boneToModify {boneName=dyn_vol_r}
    translation=<2.86 5.043>
   }
   n3 = RigidBody(ComponentPose=@n2, Alpha=0.25) {
    bClampLinearTranslationLimitToRefPose=true
    bOverrideWorldGravity=true
    overridePhysicsAsset=&/M_MED_Football_20_Physics
    simulationSpace=BaseBoneSpace
    alphaBoolBlend {blendOption=Linear}
    baseBoneRef {boneName=root}
    componentLinearAccScale=<0.1 0.1 0.1>
    overrideWorldGravity {z=-980}
   }
   n4 = RigidBody(ComponentPose=@n3, Alpha=0.25) {
    bClampLinearTranslationLimitToRefPose=true
    bOverrideWorldGravity=true
    overridePhysicsAsset=&M_MED_Football_20_Physics
    simulationSpace=WorldSpace
    alphaBoolBlend {blendOption=Linear}
    baseBoneRef {boneName=root}
    componentLinearAccScale=<0.1 0.1 0.1>
    overrideWorldGravity {z=-980}
   }
   Output Pose(Result=Component To Local(ComponentPose=@n4)) {layerGroup=None}
  ]}
 EventGraph {nodes [
   Event BlueprintUpdateAnimation()
   n1 = Get Skel_Mesh()
   > Set thigh_rot_r(thigh_rot_r=GetAlphaFromReferenceTransformDelta(SkeletalMeshComponent=@n1, BoneName=thigh_r, BaseName=pelvis, InputRange=(X=0,Y=-90), TwistAxis="0,0,1"))
   > Set thigh_rot_l(thigh_rot_l=GetAlphaFromReferenceTransformDelta(SkeletalMeshComponent=@n1, BoneName=thigh_l, BaseName=pelvis, InputRange=(X=-15,Y=-110), TwistAxis="0,0,1"))
   > n2 = Branch(Condition=Get bIsInFrontEnd())
   > Set dynalpha(dynalpha:double=0.25) -> (Output_Get:double)
   > Set rigidalpha(rigidalpha:double=0) -> (Output_Get:double)
   Set dynalpha(execute=@n2.else, dynalpha=0)
   > Set rigidalpha(rigidalpha=0.25)
   Event BlueprintInitializeAnimation()
   > Set Skel_Mesh(Skel_Mesh=GetOwnerSkeletalMeshComponent())
   TryGetPawnOwner()
  ]}
]
variables [
 dynalpha {type=number}
 rigidalpha {type=number}
 Skel_Mesh {type=SkeletalMeshComponent}
 thigh_rot_l {type=number}
 thigh_rot_r {type=number}
]
