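# Physics asset for the face accessory mesh of the male medium Football_20 character body, providing physics for its dynamic straps.
# Four thin capsules cover the left and right, high and low strap bones, and one large capsule covers the head.
# Each strap is constrained directly to the head rather than chained, with moderate swing limits and almost no twist allowed.
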PhysicsAsset /BRCosmetics/Characters/Player/Male/Medium/Bodies/M_MED_Football_20/Meshes/Parts/M_MED_Football_20_FaceAcc_Physics
bodies {
 dyn_strap_l_hi_00 [{shapeType=Capsule, length=3.505, radius=0.505, center=<0.6671 -1.5977 1.4065>, rotation=<rot -47.9156 -96.5875 19.2462>}]
 dyn_strap_l_low_00 [{shapeType=Capsule, length=4.505, radius=0.505, center=<0.6885 -1.8283 1.2644>, rotation=<rot 0 23.5972 -58.6612>}]
 dyn_strap_r_hi_00 [{shapeType=Capsule, length=2.755, radius=0.505, center=<-0.5503 -1.172 1.1296>, rotation=<rot 55.7017 97.7466 -157.9761>}]
 dyn_strap_r_low_00 [{shapeType=Capsule, length=3.755, radius=0.505, center=<-0.7374 -1.5832 1.197>, rotation=<rot -6.738 -13.6918 -60.6832>}]
 head [{shapeType=Capsule, length=13.4632, radius=6.2536, center=<5.8852 2.2599 -0.0127>, rotation=<rot -0.7785 -90 -90.6742>}]
}
constraints [
 {bone1Name=dyn_strap_r_low_00, bone2Name=head, swing1LimitDegrees=20, swing2LimitDegrees=20, twistLimitDegrees=1}
 {bone1Name=dyn_strap_r_hi_00, bone2Name=head, swing1LimitDegrees=20, swing2LimitDegrees=20, twistLimitDegrees=1}
 {bone1Name=dyn_strap_l_hi_00, bone2Name=head, swing1LimitDegrees=20, swing2LimitDegrees=20, twistLimitDegrees=1}
 {bone1Name=dyn_strap_l_low_00, bone2Name=head, swing1LimitDegrees=20, swing2LimitDegrees=20, twistLimitDegrees=1}
]
